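# Material function that outputs a tangent-space normal stamping the Atom primitive logo onto a surface.
# Fractional UVs, rescaled about the center by ScaleUVsByCenter, sample the logo's normal and mask textures with clamped samplers.
# The mask's green channel blends from a flat normal to the logo normal, gated by floored V so the logo appears only where V is below zero.
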MaterialFunction /AtomFramework/Materials/MatFunctions/MF_SimpleAtomLogo
graph {nodes [
  n1 = MaterialFunctionCall(UVs=Frac(None=TextureCoordinate()), Texture Scale=Constant() {r=0.5}) {materialFunction=&ScaleUVsByCenter}
  n2 = TextureSample(UVs=@n1.UVs) {samplerSource=SSM_Clamp_WorldGroupSettings, samplerType=SAMPLERTYPE_Normal, texture=/AtomFramework/Materials/Textures/PrimitiveLogo_N}
  n3 = TextureSample(UVs=@n1.UVs) {samplerSource=SSM_Clamp_WorldGroupSettings, samplerType=SAMPLERTYPE_Masks, texture=/AtomFramework/Materials/Textures/PrimitiveLogo_M}
  n4 = Floor(None=ComponentMask(None=TextureCoordinate()) {g=true})
  n5 = Constant() {r=1}
  n6 = Saturate(None=OneMinus(None=Add(A=@n4, B=@n5)))
  FunctionOutput(None=LinearInterpolate(A=Constant3Vector() {constant=<rgba 0 0 1 0>}, B=@n2.RGB, Alpha=Multiply(A=@n3.G, B=@n6)))
 ]}
