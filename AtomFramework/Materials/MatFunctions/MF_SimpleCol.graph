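# MF_SimpleCol is a material function that supplies a simple color from a Color vector parameter in the "1 - Col" group, red by default.
# It splits the RGB into channels, converts each from sRGB to linear, and remaps each into a narrow range just inside black and white.
# The channels are recombined into a float3 output, so the color never reaches pure black or full white.
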MaterialFunction /AtomFramework/Materials/MatFunctions/MF_SimpleCol
graph {nodes [
  n1 = MaterialFunctionCall(Float3=VectorParameter() {group=1 - Col, parameterName=Color, defaultValue=<rgba 0.7059 0 0 1>}.RGB) {materialFunction=&BreakOutFloat3Components}
  n2 = MaterialFunctionCall!(sRGB=@n1.R) {materialFunction=/Engine/Functions/Engine_MaterialFunctions01/ImageAdjustment/sRGBToLinear}
  n3 = Constant() {r=0.02}
  n4 = Constant() {r=0.95}
  n5 = MaterialFunctionCall(sRGB=@n1.G) {materialFunction=&sRGBToLinear}
  n6 = Constant() {r=0.02}
  n7 = Constant() {r=0.95}
  n8 = MaterialFunctionCall(sRGB=@n1.B) {materialFunction=&sRGBToLinear}
  n9 = Constant() {r=0.02}
  n10 = Constant() {r=0.95}
  FunctionOutput(None=MaterialFunctionCall(
    X=MaterialFunctionCall(Input=@n2.Linear, Target Low=@n3, Target High=@n4) {materialFunction=&RemapValueRange}.Result
    Y=MaterialFunctionCall(Input=@n5.Linear, Target Low=@n6, Target High=@n7) {materialFunction=&RemapValueRange}.Result
    Z=MaterialFunctionCall(Input=@n8.Linear, Target Low=@n9, Target High=@n10) {materialFunction=&RemapValueRange}.Result
   ) {materialFunction=&MakeFloat3}.Result)
 ]}
