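# M_AtomStandard_Simple is a simple-quality Default Lit material in AtomFramework, built through material attributes.
# MF_SimpleCol feeds both Base Color and Subsurface Color, and MF_SimpleAtomLogo supplies the Normal.
# A world-aligned mask texture, tiled by Roughness Scale, drives the roughness lerp.
# Both lerp inputs read Roughness Min, so roughness stays constant and Roughness Max is unused.
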Material /AtomFramework/Materials/Simple_Qual/M_AtomStandard_Simple
graph {
 properties {bUseMaterialAttributes=true, pixelDepthOffsetMode=PDOM_Legacy, usedShadingModels=Default Lit}
 nodes [
  n1 = MaterialFunctionCall() {materialFunction=/AtomFramework/Materials/MatFunctions/MF_SimpleCol}
  n2 = VectorParameter() {parameterName=Roughness Params, channelNames {a="", b=Roughness Scale, g=Roughness Max, r=Roughness Min}, defaultValue=<rgba 0.2 0.5 256 0>}
  n3 = TextureObject() {samplerType=SAMPLERTYPE_Masks, texture=/AtomFramework/Materials/Textures/Primitive_Surface_R}
  SetMaterialAttributes!(
   MaterialAttributes
   Base Color=@n1.Result
   Roughness=LinearInterpolate(
    A=@n2.Roughness Min
    B=@n2.Roughness Min
    Alpha=MaterialFunctionCall(TextureObject=@n3, TextureSize=@n2.Roughness Scale) {materialFunction=&WorldAlignedTexture}.XYZ Texture
   )
   Normal=MaterialFunctionCall() {materialFunction=/AtomFramework/Materials/MatFunctions/MF_SimpleAtomLogo}.Result
   Subsurface Color=@n1.Result
  ) {attributeSetTypes [
    #1
    #2
    #3
    #4
   ]}
 ]
}
