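# Simple, cheap translucent material for AtomFramework: two-sided, fully metallic, with screen-space reflections and per-vertex non-directional volumetric lighting.
# Material attributes take base color from MF_SimpleCol and normals from MF_SimpleAtomLogo.
# Fresnel blends refraction from 1 toward Refraction Strength.
# World-aligned mask and Fresnel lerps drive roughness and opacity, but each uses one parameter channel for both inputs, so both outputs are uniform.
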Material /AtomFramework/Materials/Simple_Qual/M_AtomTrans_Simple_Cheap
graph {
 properties {
  bCastRayTracedShadows=false
  bScreenSpaceReflections=true
  bUseMaterialAttributes=true
  blendMode=BLEND_Translucent
  pixelDepthOffsetMode=PDOM_Legacy
  refractionMethod=RM_IndexOfRefraction
  translucencyLightingMode=TLM_VolumetricPerVertexNonDirectional
  twoSided=true
  usedShadingModels=Default Lit
 }
 nodes [
  n1 = VectorParameter() {parameterName=Roughness Params, channelNames {a="", b=Roughness Scale, g=Roughness Max, r=Roughness Min}, defaultValue=<rgba 0.05 0.3 256 2>}
  n2 = TextureObject() {samplerType=SAMPLERTYPE_Masks, texture=/AtomFramework/Materials/Textures/Primitive_Surface_R}
  n3 = VectorParameter() {parameterName=Opacity, defaultValue=<rgba 0.16 1 0.5 0>}
  SetMaterialAttributes!(
   MaterialAttributes
   Base Color=MaterialFunctionCall() {materialFunction=/AtomFramework/Materials/MatFunctions/MF_SimpleCol}.Result
   Roughness=LinearInterpolate(
    A=@n1.Roughness Min
    B=@n1.Roughness Min
    Alpha=MaterialFunctionCall(TextureObject=@n2, TextureSize=@n1.Roughness Scale) {materialFunction=&WorldAlignedTexture}.XYZ Texture
   )
   Normal=MaterialFunctionCall() {materialFunction=/AtomFramework/Materials/MatFunctions/MF_SimpleAtomLogo}.Result
   Opacity=LinearInterpolate(A=@n3.R, B=@n3.R, Alpha=Fresnel(ExponentIn=@n3.B) {baseReflectFraction=0, exponent=3})
   Refraction=LinearInterpolate(B=ScalarParameter() {defaultValue=2, parameterName=Refraction Strength}, Alpha=Fresnel() {baseReflectFraction=0, exponent=3}) {constA=1, constB=1.1}
   Metallic=Constant() {r=1}
  ) {attributeSetTypes [
    #1
    #2
    #3
    #4
    #5
    #6
   ]}
 ]
}
