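# Two-sided thin translucent AtomFramework material, a simple-quality variant with full metallic, screen-space reflections and a logo normal from MF_SimpleAtomLogo.
# MF_SimpleCol drives base color and, scaled by an offset Fresnel, thin-translucent transmittance.
# A second Fresnel blends index-of-refraction toward Refraction Strength, while Opacity is a scalar parameter.
# A world-aligned mask feeds the roughness lerp, but both ends use Roughness Min, so roughness stays uniform.
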Material /AtomFramework/Materials/Simple_Qual/M_AtomTrans_Simple_Exp
graph {
 properties {
  bCastRayTracedShadows=false
  bScreenSpaceReflections=true
  bUseMaterialAttributes=true
  blendMode=BLEND_Translucent
  pixelDepthOffsetMode=PDOM_Legacy
  refractionMethod=RM_IndexOfRefraction
  shadingModel=MSM_ThinTranslucent
  translucencyLightingMode=TLM_SurfacePerPixelLighting
  twoSided=true
  usedShadingModels=Thin Translucent
 }
 nodes [
  n1 = MaterialFunctionCall() {materialFunction=/AtomFramework/Materials/MatFunctions/MF_SimpleCol}
  n2 = VectorParameter() {
   parameterName=Roughness Params
   channelNames {a="", b=Roughness Scale, g=Roughness Max, r=Roughness Min}
   defaultValue=<rgba 0.05 0.25 256 2>
  }
  n3 = TextureObject() {samplerType=SAMPLERTYPE_Masks, texture=/AtomFramework/Materials/Textures/Primitive_Surface_R}
  SetMaterialAttributes!(
   MaterialAttributes
   Base Color=@n1.Result
   Roughness=LinearInterpolate(
    A=@n2.Roughness Min
    B=@n2.Roughness Min
    Alpha=MaterialFunctionCall(TextureObject=@n3, TextureSize=@n2.Roughness Scale) {materialFunction=&WorldAlignedTexture}.XYZ Texture
   )
   Normal=MaterialFunctionCall() {materialFunction=/AtomFramework/Materials/MatFunctions/MF_SimpleAtomLogo}.Result
   Opacity=ScalarParameter() {defaultValue=0.2, parameterName=Opacity, sliderMax=1}
   Refraction=LinearInterpolate(B=ScalarParameter() {defaultValue=1.1, parameterName=Refraction Strength}, Alpha=Fresnel() {baseReflectFraction=0, exponent=3}) {constA=1, constB=1.1}
   Metallic=Constant() {r=1}
  ) {attributeSetTypes [
    #1
    #2
    #3
    #4
    #5
    #6
   ]}
  n4 = Fresnel() {baseReflectFraction=0, exponent=0.75}
  ThinTranslucentMaterialOutput!(TransmittanceColor=Multiply(A=Add(A=@n4) {constB=1.2}, B=@n1.Result), SurfaceCoverage)
 ]
}
